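# Unlit Niagara mesh-particle material for the AnvilAlarmRobin pickaxe's voxel cube FX, with an IsImpact parameter blending in impact behavior.
# UV-encoded positions drive distance masks and sine waves, which, with Progress from a dynamic parameter, displace cubes through world position offset.
# Derivative-based face normals build box-projected coordinates from normalized local position for radial gradients.
# These gradients feed two-tone color with HSV shift, angle-based reflection tint and additive pulses.
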Material /BRCosmetics/Weapons/FORT_Melee/Pickaxe_AnvilAlarmRobin/FX/M_Pickaxe_AnvilAlarmRobin_Cubes
graph {
 properties {
  bAutomaticallySetUsageInEditor=false
  bUsedWithNiagaraMeshParticles=true
  numCustomizedUVs=1
  previewMesh=&/SM_Pickaxe_AnvilAlarmRobin_VoxelImpact
  shadingModel=MSM_Unlit
  translucencyLightingMode=TLM_Surface
  usedShadingModels=Unlit
 }
 nodes [
  n1 = TextureCoordinate()
  n2 = Constant2Vector() {g=-1}
  Length(None=AppendVector(A=Add(A=@n1, B=@n2), B=Constant())) {is=class MaterialExpressionLength}
  n3 = NamedRerouteUsage()
  n4 = SmoothStep(
   Min=ScalarParameter() {defaultValue=0.4, parameterName=DistanceMaskMin, sortPriority=33}
   Max=ScalarParameter() {defaultValue=0.9, parameterName=DistanceMaskMax, sortPriority=34}
   Value=NamedRerouteUsage()
  ) {constMax=0.9, constMin=0.4}
  n5 = OneMinus(None=@n4)
  n6 = ScalarParameter() {parameterName=Progress, sliderMax=1}
  n7 = DynamicParameter() {defaultValue=<rgba 0 1 1 1>, paramNames [ColorLerp,Progress,Param3,Param4]}
  n8 = LinearInterpolate(A=@n6, B=@n7.Progress, Alpha=Step(Y=@n6) {constX=0.01})
  n9 = LinearInterpolate(Alpha=@n8) {constA=-1}
  n10 = Sine(None=Add(A=Multiply(A=@n9) {constB=-1}, B=NamedRerouteUsage())) {period=2.5}
  n11 = OneMinus(None=Saturate(None=SmoothStep(Value=@n10) {constMax=0.6}))
  n12 = ScalarParameter() {parameterName=IsImpact}
  n13 = ScalarParameter() {defaultValue=1, parameterName=VertexAlphaWPO Multiplier}
  n14 = Multiply(A=LinearInterpolate(A=Power(Base=@n5) {constExponent=5}, B=Multiply(A=@n11) {constB=-1}, Alpha=@n12) {constB=0}, B=@n13)
  n15 = ScalarParameter() {defaultValue=-8.0593, parameterName=MinWPO}
  n16 = Abs(None=@n15)
  n17 = Saturate(None=OneMinus(None=@n8))
  n18 = LinearInterpolate(A=Multiply(A=Add(A=@n16, B=@n13), B=Multiply(A=@n17) {constA=1, constB=-1}), Alpha=@n12)
  n19 = Multiply(A=@n8) {constB=1.2}
  n20 = LinearInterpolate(Alpha=Saturate(None=Subtract(A=@n19) {constB=0.2})) {constA=1, constB=-1}
  n21 = NamedRerouteUsage()
  n22 = Power(Base=Sine(None=Add(A=@n20, B=@n21) {constA=1, constB=0.5}) {period=2}) {constExponent=20}
  n23 = ScalarParameter() {parameterName=IsImpact}
  n24 = Add(A=Clamp(None=Add(A=@n14, B=@n18), Min=@n15) {maxDefault=0}, B=LinearInterpolate(A=Multiply(A=@n22) {constB=2}, Alpha=@n23) {constB=0})
  n25 = SmoothStep(Value=@n10)
  n26 = Multiply(
   A=Transform(None=Constant3Vector() {constant=<rgba 0 0 1 0>}) {transformSourceType=TRANSFORMSOURCE_Instance}
   B=Multiply(A=Power(Base=@n25) {constExponent=30}) {constB=12}
  )
  LinearInterpolate(
   A=Add(A=Multiply(A=@n3, B=@n24), B=LinearInterpolate(B=@n26, Alpha=@n23) {constB=0})
   Alpha=ScalarParameter() {parameterName=UseNormalTexCoord}
  ) {constB=0}
  n27 = Saturate(None=MaterialFunctionCall(
    UVs=NamedRerouteUsage()
    Radius=ScalarParameter() {defaultValue=0.8, parameterName=RadialGradientRadius}
    Density=ScalarParameter() {defaultValue=1, parameterName=RadialGradientDensity}
   ) {materialFunction=&RadialGradientExponential}.RadialGradientExponential)
  n28 = LinearInterpolate(
   A=VectorParameter() {parameterName=OuterColor, defaultValue=<rgba 0.08 0.0257 0.1667 1>}.RGB
   B=VectorParameter() {parameterName=InnerColor, defaultValue=<rgba 0.2299 0.0378 0.8438 1>}.RGB
   Alpha=@n27
  )
  n29 = MaterialFunctionCall(Color=Add(A=RgbToHsv!(None=@n28), B=VectorParameter() {parameterName=HSV Shift, defaultValue=<rgba -0.036 0.0204 -0.1456 1>}.RGB)) {materialFunction=&SplitComponents}
  n30 = HsvToRgb!(None=MaterialFunctionCall(R=Frac(None=@n29.R), G=@n29.G, B=@n29.G) {materialFunction=&AppendMany}.RGB)
  n31 = Add(A=Power(Base=@n4, Exp=ScalarParameter() {defaultValue=1, parameterName=VertexAlphaPower}), B=@n7.ColorLerp) {constB=0}
  n32 = VectorParameter() {parameterName=ReflectionColor, defaultValue=<rgba 0.1736 0.4929 0.901 1>}
  n33 = Saturate(None=OneMinus(None=NamedRerouteUsage()))
  n34 = Multiply(
   A=ParticleColor().RGB
   B=LinearInterpolate(A=LinearInterpolate(A=@n28, B=@n30, Alpha=@n31), B=Multiply(A=@n28, B=@n32.RGB), Alpha=SmoothStep(Value=@n33) {constMax=0.2})
  )
  n35 = LinearInterpolate(B=NamedRerouteUsage(), Alpha=ScalarParameter() {defaultValue=1, parameterName=Add Pulsing})
  LinearInterpolate(
   A=MaterialFunctionCall(Emissive=Add(A=@n34, B=@n35)) {materialFunction=&MF_FX_EmissiveAdjustments}.Result
   B=NamedRerouteUsage()
  ) {constAlpha=0}
  Multiply(A=Frac(None=Time()))
  n36 = AppendVector(A=Add(A=TextureCoordinate(), B=Constant2Vector() {g=-1}), B=ComponentMask(None=TextureCoordinate() {coordinateIndex=1}) {r=true})
  Multiply(
   A=@n36
   B=MaterialFunctionCall() {materialFunction=&ObjectScale}.Scale XYZ
  )
  n37 = SmoothStep(Value=NamedRerouteUsage()) {constMax=0.95, constMin=0.6}
  n38 = Multiply(A=Power(Base=@n37) {constExponent=1})
  n39 = Multiply(A=Time()) {constB=-1.5}
  n40 = Multiply(A=@n37)
  n41 = Power(Base=Sine(None=Add(A=@n39, B=@n40)) {period=2}) {constExponent=5}
  n42 = Multiply(
   A=LinearInterpolate(
    A=VectorParameter() {parameterName=PulseColor_1, defaultValue=<rgba 1.8353 0.325 2 1>}.RGB
    B=VectorParameter() {parameterName=PulseColor_2, defaultValue=<rgba 0 1.1881 4 1>}.RGB
    Alpha=Saturate(None=@n38)
   )
   B=LinearInterpolate(A=Saturate(None=@n41), B=NamedRerouteUsage(), Alpha=ScalarParameter() {parameterName=IsImpact}) {constB=0}
  )
  n43 = NamedRerouteUsage()
  n44 = LinearInterpolate(Alpha=OneMinus(None=MaterialFunctionCall(UVs=@n43) {materialFunction=&RadialGradientExponential}.RadialGradientExponential)) {constA=0.4}
  NamedRerouteDeclaration(None=Multiply(A=Multiply(A=@n42, B=@n44))) {name=Pulse, nodeColor=<rgba 0.2188 1 0 1>}
  NamedRerouteDeclaration(None=Power(Base=@n25) {constExponent=20}) {name=Impact Pulse, nodeColor=<rgba 1 0 0.1406 1>}
  n45 = Abs(None=DotProduct(A=CameraVectorWS(), B=NamedRerouteUsage())) {desc=Make this work on two-sided materials}
  NamedRerouteDeclaration(None=Clamp(None=Power(Base=@n45) {constExponent=0.2})) {name=AngleFade, nodeColor=<rgba 1 0 0.4219 1>}
  NamedRerouteDeclaration(None=Length(None=@n36)) {name=Distance, nodeColor=<rgba 0 0.7422 1 1>}
  n46 = WorldPosition() {worldPositionShaderOffset=WPT_ExcludeAllShaderOffsets}
  n47 = CrossProduct(A=DDX(Value=@n46.XYZ), B=DDY(Value=@n46.XYZ))
  NamedRerouteDeclaration(None=Normalize(VectorInput=Transform(None=@n47) {transformSourceType=TRANSFORMSOURCE_World, transformType=TRANSFORM_Instance})) {name=Local Normal, nodeColor=<rgba 0 1 0.1328 1>}
  NamedRerouteDeclaration(None=Normalize(VectorInput=@n47)) {name=World Normal, nodeColor=<rgba 0 1 0.8828 1>}
  NamedRerouteDeclaration(None=@n27) {name=_Debug, nodeColor=<rgba 1 0 0.8672 1>}
  n48 = NamedRerouteUsage()
  n49 = Abs(None=DotProduct(A=Constant3Vector() {constant=<rgba 1 0 0 0>}, B=@n48))
  n50 = Abs(None=DotProduct(A=@n48, B=Constant3Vector() {constant=<rgba 0 1 0 0>}))
  n51 = OneMinus(None=Add(A=@n49, B=@n50))
  n52 = LocalPosition() {includedOffsets=ExcludeOffsets}
  n53 = ScalarParameter() {defaultValue=1.5, parameterName=LocalPosDividor}
  n54 = Normalize(VectorInput=Subtract(A=Divide(A=@n52.XYZ, B=@n53) {constB=1.5}, B=Multiply(A=@n36) {constB=100}) {constB=0})
  n55 = ComponentMask(None=@n54) {r=true}
  n56 = ComponentMask(None=@n54) {g=true}
  n57 = AppendVector(A=@n55, B=@n56)
  n58 = ComponentMask(None=@n54) {b=true}
  n59 = AppendVector(A=@n56, B=@n58)
  n60 = Add(A=Add(A=Multiply(A=@n51, B=@n57), B=Multiply(A=@n49, B=@n59)), B=Multiply(A=@n50, B=AppendVector(A=@n55, B=@n58)))
  NamedRerouteDeclaration(None=LinearInterpolate(A=Add(A=@n60) {constB=0.5}, B=TextureCoordinate(), Alpha=ScalarParameter() {parameterName=UseNormalTexCoord})) {name=Custom TexCoord, nodeColor=<rgba 1 0.2344 0 1>}
  NamedRerouteDeclaration(None=Transform(None=@n54) {transformSourceType=TRANSFORMSOURCE_Instance}) {name=DirectionalVector, nodeColor=<rgba 1 0.5391 0 1>}
  ScalarParameter() {parameterName=ImpactProgress}
 ]
}
